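# Conversation database for the NPC Ragsy, giving the dialogue entry point for talking to this character.
# The entry leads to a speech node with a default line and context-driven alternatives, prioritized by curve rows for time since last match, first meeting, held bars, and a specific outfit.
# The speech node then branches through link nodes into the shared NPC quest and basic services conversations.
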ConversationDatabase /NPCLibrary/NPCs/Ragsy/Ragsy/NPCLibrary_Conversation_Ragsy
entry_points [{tag=Conversation.Entry.Ragsy, destinations [#1]}]
nodes [
 ConversationEntryPointNode_7 {
  id=#1
  class=ConversationEntryPointNode
  outputs [#2]
  properties {entryTag {tagName=Conversation.Entry.Ragsy}}
 }
 Speech_C_4 {
  id=#2
  class=Speech_C
  outputs [#3, #4]
  properties {
   nodeName=""
   generalConfig {
    defaultMessage=Please tell me you've been shielding up.
    contextualMessages [
     {
      message="You're putting yourself back into the line of fire?!"
      contextRequirements [{requirement {class=FortControllerRequirement_CheckTimeSinceLastPlayedMatch}}]
      priority {curve {rowName=Tandem.Speech.Priority.WhileSincePlayed}}
     }
     {
      message="Safety is the name of the game! "
      contextRequirements [{requirement {class=FortControllerRequirement_IsFirstEverConversationWithNPC}}]
      priority {curve {rowName=Tandem.Speech.Priority.FirstEverInteraction}}
     }
     {
      message="You should stock up on shields. Like, now."
      contextRequirements [{requirement {class=FortControllerRequirement_HasBars}}]
      priority {curve {rowName=Tandem.Speech.Priority.HighTotalBars}}
     }
     {
      message="You look like I did, before... *IT* happened."
      contextRequirements [{requirement {
         class=FortControllerRequirement_HasCID
         properties {allowedCIds [/Game/Athena/Items/Cosmetics/Characters/CID_069_Athena_Commando_F_PinkBear]}
        }}]
      priority {curve {rowName=Tandem.Speech.Priority.OutfitReaction}}
     }
    ]
   }
  }
 }
 ConversationLinkNode_0 {id=#4, class=ConversationLinkNode, properties {remoteEntryTag {tagName=Conversation.Entry.BasicServices}}}
 ConversationLinkNode_1 {id=#3, class=ConversationLinkNode, properties {remoteEntryTag {tagName=Conversation.Entry.NPCLibrary.Quests}}}
]
